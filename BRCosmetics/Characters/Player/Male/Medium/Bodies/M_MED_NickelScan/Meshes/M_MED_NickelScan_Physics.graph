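# Physics asset for the NickelScan medium male player body, setting up capsule bodies and joints for its dangling skirt bones.
# Five skirt chains (A to E), each two to four segments below a small base body, hang from the pelvis on free-rotating base joints.
# Segment joints allow very little twist and mostly wider swing on one axis.
# Pelvis, both thighs and only the right calf also have capsules, with no constraints between them.
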PhysicsAsset /BRCosmetics/Characters/Player/Male/Medium/Bodies/M_MED_NickelScan/Meshes/M_MED_NickelScan_Physics
bodies {
 calf_r [{shapeType=Capsule, length=44.625, radius=7, shapeName=thigh_r_capsule, center=<23.3967 2.456 -0.4471>, rotation=<rot -90 179.9998 0>}]
 dyn_skirt_down_A_01 [{shapeType=Capsule, length=11.375, radius=7, shapeName=dyn_skirt_down_A_01_capsule, center=<7 0 7>, rotation=<rot 0.0001 0 75>}]
 dyn_skirt_down_A_02 [{
   shapeType=Capsule
   length=12.625
   radius=7
   shapeName=dyn_skirt_down_A_02_capsule
   center=<7.8084 1.105 5.785>
   rotation=<rot 0.0003 -0.0002 65>
  }]
 dyn_skirt_down_A_03 [{
   shapeType=Capsule
   length=8.5
   radius=7
   shapeName=dyn_skirt_down_A_03_capsule
   center=<8.6459 1.4943 3.6729>
   rotation=<rot 0.0002 -0.0001 75>
  }]
 dyn_skirt_down_A_Base [{shapeType=Capsule, length=2, radius=2, shapeName=dyn_skirt_down_A_Base_capsule, rotation=<rot 90 0 0>}]
 dyn_skirt_down_B_01 [{shapeType=Capsule, length=11.375, radius=7, shapeName=dyn_skirt_down_B_01_capsule, center=<3.5 0 7>, rotation=<rot 0 0 90>}]
 dyn_skirt_down_B_02 [{shapeType=Capsule, length=11.375, radius=7, shapeName=dyn_skirt_down_B_02_capsule, center=<8 0 7>, rotation=<rot 0 0 90>}]
 dyn_skirt_down_B_03 [{shapeType=Capsule, length=11.375, radius=7, shapeName=dyn_skirt_down_B_03_capsule, center=<9 0 7>, rotation=<rot 0 0 90>}]
 dyn_skirt_down_B_04 [{shapeType=Capsule, length=11.375, radius=7, shapeName=dyn_skirt_down_B_04_capsule, center=<10 0 7>, rotation=<rot 0 0 90>}]
 dyn_skirt_down_B_Base [{shapeType=Capsule, length=2, radius=2, shapeName=dyn_skirt_down_B_Base_capsule, rotation=<rot 90 0 0>}]
 dyn_skirt_down_C_01 [{shapeType=Capsule, length=11.375, radius=7, shapeName=dyn_skirt_down_C_01_capsule, center=<7 0 7>, rotation=<rot 0 0 90>}]
 dyn_skirt_down_C_02 [{shapeType=Capsule, length=11.375, radius=7, shapeName=dyn_skirt_down_C_02_capsule, center=<8 0 7>, rotation=<rot 0 0 90>}]
 dyn_skirt_down_C_03 [{shapeType=Capsule, length=11.375, radius=7, shapeName=dyn_skirt_down_C_03_capsule, center=<9 0 5.9568>, rotation=<rot 0 0 90>}]
 dyn_skirt_down_C_Base [{shapeType=Capsule, length=2, radius=2, shapeName=dyn_skirt_down_C_Base_capsule, rotation=<rot 90 0 0>}]
 dyn_skirt_down_D_01 [{shapeType=Capsule, length=11.375, radius=7, shapeName=dyn_skirt_down_D_01_capsule, center=<7 0 7>, rotation=<rot 0 0 90>}]
 dyn_skirt_down_D_02 [{shapeType=Capsule, length=11.375, radius=7, shapeName=dyn_skirt_down_D_02_capsule, center=<8 0 7>, rotation=<rot 0 0 90>}]
 dyn_skirt_down_D_Base [{shapeType=Capsule, length=2, radius=2, shapeName=dyn_skirt_down_D_Base_capsule, rotation=<rot 90 0 0>}]
 dyn_skirt_down_E_01 [{shapeType=Capsule, length=11.375, radius=7, shapeName=dyn_skirt_down_E_01_capsule, center=<7 0 6.6873>, rotation=<rot 0 0 90>}]
 dyn_skirt_down_E_02 [{shapeType=Capsule, length=11.375, radius=7, shapeName=dyn_skirt_down_E_02_capsule, center=<8 0 6.6873>, rotation=<rot 0 0 90>}]
 dyn_skirt_down_E_Base [{shapeType=Capsule, length=2, radius=2, shapeName=dyn_skirt_down_E_Base_capsule, rotation=<rot 90 0 0>}]
 pelvis [{shapeType=Capsule, length=6.5, radius=10, shapeName=pelvis_capsule, center=<-1.9798 -1.6911 0>}]
 thigh_l [{shapeType=Capsule, length=37.625, radius=7.75, shapeName=thigh_l_capsule, center=<-19.2454 -1.1779 1.6146>, rotation=<rot -90 -0.0002 0>}]
 thigh_r [{
   shapeType=Capsule
   length=36.375
   radius=9
   shapeName=thigh_r_capsule
   center=<19.3035 1.1438 -1.7082>
   rotation=<rot -87.6554 -0.103 179.2766>
  }]
}
constraints [
 {
  bone1Name=dyn_skirt_down_A_Base
  bone2Name=pelvis
  swing1LimitDegrees=45
  swing1Motion=Free
  swing2LimitDegrees=45
  swing2Motion=Free
  twistLimitDegrees=45
  twistMotion=Free
 }
 {bone1Name=dyn_skirt_down_A_01, bone2Name=dyn_skirt_down_A_Base, swing1LimitDegrees=5, swing2LimitDegrees=15, twistLimitDegrees=2}
 {bone1Name=dyn_skirt_down_A_02, bone2Name=dyn_skirt_down_A_01, swing1LimitDegrees=5, swing2LimitDegrees=15, twistLimitDegrees=2}
 {bone1Name=dyn_skirt_down_A_03, bone2Name=dyn_skirt_down_A_02, swing1LimitDegrees=5, swing2LimitDegrees=15, twistLimitDegrees=2}
 {
  bone1Name=dyn_skirt_down_B_Base
  bone2Name=pelvis
  swing1LimitDegrees=45
  swing1Motion=Free
  swing2LimitDegrees=45
  swing2Motion=Free
  twistLimitDegrees=45
  twistMotion=Free
 }
 {bone1Name=dyn_skirt_down_B_01, bone2Name=dyn_skirt_down_B_Base, swing1LimitDegrees=5, swing2LimitDegrees=5, twistLimitDegrees=2}
 {bone1Name=dyn_skirt_down_B_02, bone2Name=dyn_skirt_down_B_01, swing1LimitDegrees=5, swing2LimitDegrees=30, twistLimitDegrees=2}
 {bone1Name=dyn_skirt_down_B_03, bone2Name=dyn_skirt_down_B_02, swing1LimitDegrees=5, swing2LimitDegrees=30, twistLimitDegrees=2}
 {bone1Name=dyn_skirt_down_B_04, bone2Name=dyn_skirt_down_B_03, swing1LimitDegrees=5, swing2LimitDegrees=30, twistLimitDegrees=2}
 {
  bone1Name=dyn_skirt_down_C_Base
  bone2Name=pelvis
  swing1LimitDegrees=45
  swing1Motion=Free
  swing2LimitDegrees=45
  swing2Motion=Free
  twistLimitDegrees=45
  twistMotion=Free
 }
 {bone1Name=dyn_skirt_down_C_01, bone2Name=dyn_skirt_down_C_Base, swing1LimitDegrees=5, swing2LimitDegrees=30, twistLimitDegrees=2}
 {bone1Name=dyn_skirt_down_C_02, bone2Name=dyn_skirt_down_C_01, swing1LimitDegrees=5, swing2LimitDegrees=30, twistLimitDegrees=2}
 {bone1Name=dyn_skirt_down_C_03, bone2Name=dyn_skirt_down_C_02, swing1LimitDegrees=5, swing2LimitDegrees=30, twistLimitDegrees=2}
 {
  bone1Name=dyn_skirt_down_D_Base
  bone2Name=pelvis
  swing1LimitDegrees=45
  swing1Motion=Free
  swing2LimitDegrees=45
  swing2Motion=Free
  twistLimitDegrees=45
  twistMotion=Free
 }
 {bone1Name=dyn_skirt_down_D_01, bone2Name=dyn_skirt_down_D_Base, swing1LimitDegrees=5, swing2LimitDegrees=30, twistLimitDegrees=2}
 {bone1Name=dyn_skirt_down_D_02, bone2Name=dyn_skirt_down_D_01, swing1LimitDegrees=5, swing2LimitDegrees=30, twistLimitDegrees=2}
 {
  bone1Name=dyn_skirt_down_E_Base
  bone2Name=pelvis
  swing1LimitDegrees=45
  swing1Motion=Free
  swing2LimitDegrees=45
  swing2Motion=Free
  twistLimitDegrees=45
  twistMotion=Free
 }
 {bone1Name=dyn_skirt_down_E_01, bone2Name=dyn_skirt_down_E_Base, swing1LimitDegrees=5, swing2LimitDegrees=30, twistLimitDegrees=2}
 {bone1Name=dyn_skirt_down_E_02, bone2Name=dyn_skirt_down_E_01, swing1LimitDegrees=5, swing2LimitDegrees=30, twistLimitDegrees=2}
]
