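# Animation Blueprint for the NickelScan medium male player body, adding secondary motion to its bag, strips and other physics bones.
# The parent mesh pose passes through Control Rig, bone transforms and AnimDynamics into a cached pose feeding three RigidBody simulations.
# Skydive float-down and long back bling select among skydive, cape and default physics assets.
# The event graph updates gravity state, Perlin-modulated sine strip rotations, and a damped alpha that drops during movement.
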AnimBlueprint /BRCosmetics/Characters/Player/Male/Medium/Bodies/M_MED_NickelScan/Meshes/M_MED_NickelScan_AnimBP
target_skeleton=&/M_MED_NickelScan_Skeleton
graphs [
 AnimGraph {nodes [
   n1 = RigidBody!(
    ComponentPose=Local To Component(LocalPose=Use cached pose 'AnimDynamicPose'())
    Alpha
    OverrideWorldGravity:Vector="0,0,-980"
    ExternalForce
    SimSpaceSettings:Sim Space Settings Structure=(WorldAlpha=1,VelocityScaleZ=0,DampingAlpha=1,MaxLinearVelocity=200,MaxAngularVelocity=8,MaxLinearAcceleration=1000,MaxAngularAcceleration=40,ExternalLinearDragV=(X=0,Y=0,Z=0),ExternalLinearVelocity=(X=0,Y=0,Z=0),ExternalAngularVelocity=(X=0,Y=0,Z=0))
   ) -> (Pose) {
    bOverrideWorldGravity=true
    overridePhysicsAsset=&/M_MED_NickelScan_Physics
    simulationSpace=BaseBoneSpace
    baseBoneRef {boneName=spine_01}
    overrideWorldGravity {z=-980}
    simSpaceSettings {maxAngularAcceleration=40, maxAngularVelocity=8, maxLinearAcceleration=1000, maxLinearVelocity=200, velocityScaleZ=0, worldAlpha=1}
   }
   n2 = Component To Local(ComponentPose=@n1)
   n3 = RigidBody(
    ComponentPose=Local To Component(LocalPose=Use cached pose 'AnimDynamicPose'())
    OverrideWorldGravity="0,0,-980"
    SimSpaceSettings=(WorldAlpha=1,VelocityScaleZ=0,DampingAlpha=1,MaxLinearVelocity=200,MaxAngularVelocity=8,MaxLinearAcceleration=1000,MaxAngularAcceleration=40,ExternalLinearDragV=(X=0,Y=0,Z=0),ExternalLinearVelocity=(X=0,Y=0,Z=0),ExternalAngularVelocity=(X=0,Y=0,Z=0))
   ) {
    bOverrideWorldGravity=true
    overridePhysicsAsset=&/M_MED_NickelScan_WithCape_Physics
    simulationSpace=BaseBoneSpace
    baseBoneRef {boneName=spine_01}
    overrideWorldGravity {z=-980}
    simSpaceSettings {maxAngularAcceleration=40, maxAngularVelocity=8, maxLinearAcceleration=1000, maxLinearVelocity=200, velocityScaleZ=0, worldAlpha=1}
   }
   n4 = Component To Local(ComponentPose=@n3)
   n5 = Get Is Backbling Long() -> (Is Backbling Long:bool)
   n6 = RigidBody(
    ComponentPose=Local To Component(LocalPose=Use cached pose 'AnimDynamicPose'())
    OverrideWorldGravity="0,0,-980"
    SimSpaceSettings=(WorldAlpha=1,VelocityScaleZ=0,DampingAlpha=1,MaxLinearVelocity=200,MaxAngularVelocity=8,MaxLinearAcceleration=1000,MaxAngularAcceleration=40,ExternalLinearDragV=(X=0,Y=0,Z=0),ExternalLinearVelocity=(X=0,Y=0,Z=0),ExternalAngularVelocity=(X=0,Y=0,Z=0))
   ) {
    bOverrideWorldGravity=true
    overridePhysicsAsset=&/M_MED_NickelScan_SkyDive_Physics
    simulationSpace=BaseBoneSpace
    baseBoneRef {boneName=spine_01}
    overrideWorldGravity {z=-980}
    simSpaceSettings {maxAngularAcceleration=40, maxAngularVelocity=8, maxLinearAcceleration=1000, maxLinearVelocity=200, velocityScaleZ=0, worldAlpha=1}
   }
   Output Pose(Result=Blend Poses by bool(
     BlendPose_0=Component To Local(ComponentPose=@n6)
     BlendPose_1=Blend Poses by bool(BlendPose_0=@n4, BlendPose_1=@n2, bActiveValue=@n5)
     bActiveValue=Get IsSkyDiveFloatingDown() -> (IsSkyDiveFloatingDown:bool)
    )) {layerGroup=None}
   n7 = Control Rig(Source=Copy Pose From Mesh() {bUseAttachedParent=true}) {controlRigAssetReference {blueprintRigClass=&/M_MED_NickelScan_CtrlRig.M_MED_NickelScan_CtrlRig_C}}
   n8 = Anim Dynamics - Bone: dyn_bag_01!(ComponentPose=Local To Component(LocalPose=@n7), Alpha=1, GravityOverride:Vector="0,0,-980", ExternalForce="0,0,0") -> (Pose) {
    angularBiasOverride=0.999
    angularDampingOverride=0.9999
    angularSpringConstant=1000
    bGravityOverrideInSimSpace=true
    bOverrideAngularBias=true
    bOverrideAngularDamping=true
    bUseGravityOverride=true
    simulationSpace=RootRelative
    boundBone {boneName=dyn_bag_01}
    gravityOverride {z=-980}
    physicsBodyDefinitions [{
      boundBone {boneName=dyn_bag_01}
      boxExtents=<10 10 10>
      constraintSetup {angularLimitsMax=<0 3 3>, angularLimitsMin=<0 -3 -20>, angularTarget=<0 0 0>}
      localJointOffset=<3 0 0>
     }]
    relativeSpaceBone {boneName=thigh_l}
    simSpaceSettings {maxAngularAcceleration=36, maxAngularVelocity=6, simSpaceAngularAlpha=1}
   }
   n9 = "Transform (Modify) Bone - Bone: dyn_strip_shoulder_Base"(ComponentPose=@n8, Alpha=1, Translation="0,0,0", Rotation="0,0,0", Scale="1,1,1") {rotationMode=BMM_Additive, rotationSpace=BCS_BoneSpace, boneToModify {boneName=dyn_strip_shoulder_Base}}
   n10 = Anim Dynamics - Bone: dyn_strip_shoulder_01(
    ComponentPose="Transform (Modify) Bone - Bone: dyn_strip_leg_Base"(
     ComponentPose="Transform (Modify) Bone - Bone: dyn_strip_waist_Base"(ComponentPose=@n9, Alpha=1, Translation="0,0,0", Rotation="0,0,0", Scale="1,1,1") {rotationMode=BMM_Additive, rotationSpace=BCS_BoneSpace, boneToModify {boneName=dyn_strip_waist_Base}}
     Alpha=1
     Translation="0,0,0"
     Rotation="0,0,0"
     Scale="1,1,1"
    ) {rotationMode=BMM_Additive, rotationSpace=BCS_BoneSpace, boneToModify {boneName=dyn_strip_leg_Base}}
    Alpha=1
    ExternalForce="0,0,0"
   ) {
    angularBiasOverride=0.999
    angularDampingOverride=0.9999
    angularSpringConstant=1000
    bGravityOverrideInSimSpace=true
    bOverrideAngularBias=true
    bOverrideAngularDamping=true
    bUseGravityOverride=true
    simulationSpace=RootRelative
    boundBone {boneName=dyn_strip_shoulder_01}
    gravityOverride {z=500}
    physicsBodyDefinitions [{
      boundBone {boneName=dyn_strip_shoulder_01}
      boxExtents=<30 30 30>
      constraintSetup {angularLimitsMax=<0 10 10>, angularLimitsMin=<0 -10 -10>, angularTarget=<0 0 0>}
      localJointOffset=<5 0 0>
     }]
    relativeSpaceBone {boneName=clavicle_r}
    simSpaceSettings {maxAngularAcceleration=60, maxAngularVelocity=12, simSpaceAngularAlpha=1}
   }
   n11 = Anim Dynamics - Bone: dyn_strip_waist_01(ComponentPose=@n10, Alpha=1, ExternalForce="0,0,0") {
    angularDampingOverride=1
    angularSpringConstant=3000
    bAngularSpring=true
    bGravityOverrideInSimSpace=true
    bOverrideAngularDamping=true
    bUseGravityOverride=true
    gravityScale=0.35
    simulationSpace=RootRelative
    alphaBoolBlend {blendOption=Linear}
    boundBone {boneName=dyn_strip_waist_01}
    componentAppliedLinearAccClamp=<2500 2500 2500>
    componentLinearAccScale=<0.6 0.6 0.6>
    componentLinearVelScale=<0.3 3 0.3>
    gravityOverride {z=-980}
    physicsBodyDefinitions [{
      boundBone {boneName=dyn_strip_waist_01}
      boxExtents=<45 15 15>
      constraintSetup {angularLimitsMax=<0 5 5>, angularLimitsMin=<0 -10 -5>, angularTarget=<1 0 0>, linearAxesMax=<0 0.25 0.2>, linearAxesMin=<0 -0.25 -0.2>}
      localJointOffset=<3 0 0>
     }]
   }
   n12 = Anim Dynamics - Bone: dyn_strip_leg_01(ComponentPose=@n11, Alpha=1, ExternalForce="0,0,0") {
    angularBiasOverride=0.999
    angularDampingOverride=0.9999
    angularSpringConstant=1000
    bGravityOverrideInSimSpace=true
    bOverrideAngularBias=true
    bOverrideAngularDamping=true
    bUseGravityOverride=true
    simulationSpace=RootRelative
    boundBone {boneName=dyn_strip_leg_01}
    gravityOverride {y=500}
    physicsBodyDefinitions [{
      boundBone {boneName=dyn_strip_leg_01}
      boxExtents=<30 30 30>
      constraintSetup {angularLimitsMax=<0 10 10>, angularLimitsMin=<0 -10 -10>, angularTarget=<0 0 0>}
      localJointOffset=<3 0 0>
     }]
    relativeSpaceBone {boneName=dyn_strip_leg_Base}
    simSpaceSettings {maxAngularAcceleration=60, maxAngularVelocity=12, simSpaceAngularAlpha=1}
   }
   n13 = Anim Dynamics - Bone: dyn_strip_waist_02(ComponentPose=@n12, Alpha=1, ExternalForce="0,0,0") {
    angularDampingOverride=1
    angularSpringConstant=3000
    bAngularSpring=true
    bGravityOverrideInSimSpace=true
    bOverrideAngularDamping=true
    bUseGravityOverride=true
    gravityScale=0.35
    simulationSpace=RootRelative
    alphaBoolBlend {blendOption=Linear}
    boundBone {boneName=dyn_strip_waist_02}
    componentAppliedLinearAccClamp=<2500 2500 2500>
    componentLinearAccScale=<0.6 0.6 0.6>
    componentLinearVelScale=<0.3 3 0.3>
    gravityOverride {z=-980}
    physicsBodyDefinitions [{
      boundBone {boneName=dyn_strip_waist_02}
      boxExtents=<45 15 15>
      constraintSetup {angularLimitsMax=<0 5 5>, angularLimitsMin=<0 -10 -5>, angularTarget=<1 0 0>, linearAxesMax=<0 0.25 0.2>, linearAxesMin=<0 -0.25 -0.2>}
      localJointOffset=<3 0 0>
     }]
   }
   Save cached pose 'AnimDynamicPose'(Pose=Component To Local(ComponentPose=@n13)) {cacheName=AnimDynamicPose}
  ]}
 EventGraph {nodes [
   n1 = Event BlueprintUpdateAnimation()
   > Set Delta Time X(Delta Time X=@n1.DeltaTimeX)
   > n2 = Sequence!(execute) -> (then_0, then_1, then_2:Exec, then_3:Exec, then_4:Exec)
   > n3 = MakeBoneRotateWave!(
    execute
    self
    Frequencey:Vector="0.8,1.5,1"
    "X (Roll)":Vector 2D Structure=(X=-3,Y=3)
    "Y (Pitch)":Vector 2D Structure=(X=-40,Y=25)
    "Z (Yaw)":Vector 2D Structure=(X=-20,Y=20)
   ) -> (then, Rotate:Rotator)
   > Set Strip_Shoulder_base_Rotation(Strip_Shoulder_base_Rotation:Rotator=@n3) -> (Output_Get:Rotator)
   > n4 = MakeBoneRotateWave(Frequencey="0.8,1,0.6", "X (Roll)"=(X=-10,Y=10), "Y (Pitch)"=(X=-2.5,Y=40), "Z (Yaw)"=(X=-10,Y=10))
   > Set Strip_Waist_base_Rotation(Strip_Waist_base_Rotation:Rotator=@n4) -> (Output_Get:Rotator)
   > n5 = MakeBoneRotateWave(Frequencey="1,1.5,0.8", "Y (Pitch)"=(X=-15,Y=15), "Z (Yaw)"=(X=0,Y=20))
   > Set Strip_Leg_base_Rotation(Strip_Leg_base_Rotation:Rotator=@n5) -> (Output_Get:Rotator)
   n6 = "Gravity Override State Function (Parameter Asset) (Nativized)"(
    execute=@n2.then_2
    CharacterPartAnimInstance=Self-Reference()
    GameTimeInSeconds=GetGameTimeInSeconds()
    Parameters=&/M_MED_NickelScan_Parameters
   )
   > Set Gravity Override(Gravity Override=@n6)
   > Set RBAN Settings Out(RBAN Settings Out=@n6.OutRigidBodyAnimNodeParameters)
   > Set IsSkyDiveFloatingDown(IsSkyDiveFloatingDown:bool="Equal (Name)"(A=@n6.OutCurrentState, B=Skydive_FloatingDown)) -> (Output_Get:bool)
   Set Is Backbling Long(execute=@n2.then_3, Is Backbling Long:bool=CheckForBackbling().bIsBackblingLong) -> (Output_Get:bool)
   n7 = Get bIsMoving2D()
   Set Strip_Rotate_Alpha(
    execute=@n2.then_1
    Strip_Rotate_Alpha:float=DampFloat!(
     Value:float=Get Strip_Rotate_Alpha() -> (Strip_Rotate_Alpha:float)
     Target:float=SelectFloat(A=0, B=1, bPickA=@n7)
     DeltaTime:float=Get Delta Time X()
     SmoothingTime:float=0.1
    ) -> (ReturnValue:float)
   ) -> (Output_Get:float)
  ]}
 GetTimeSineNoise {nodes [
   n1 = GetTimeSineNoise!() -> (then, frequencey:double=~) {is=title K2Node_FunctionEntry}
   n2 = float * float(A=@n1, B=GetGameTimeInSeconds())
   n3 = PerlinNoise1D!(Value:float=@n2) -> (ReturnValue:float)
   > Return Node!(
    execute
    output:double=float * float(A="Sin (Radians)"(A=@n2), B=Remap Float(Input=@n3, OldRange=(X=-1,Y=1), NewRange=(X=0.5,Y=1)) -> (Output=~)) {is=title K2Node_PromotableOperator}
   )
  ]}
 MakeBoneRotateWave {nodes [
   n1 = MakeBoneRotateWave!() -> (then, Frequencey:Vector=~, "X (Roll)":Vector 2D Structure, "Y (Pitch)":Vector 2D Structure, "Z (Yaw)":Vector 2D Structure) {is=title K2Node_FunctionEntry}
   n2 = BreakVector(InVec=@n1)
   > n3 = GetTimeSineNoise!(execute, self, frequencey:double=@n2) -> (then, output:double) {is=title K2Node_CallFunction}
   > n4 = GetTimeSineNoise(frequencey=@n2.Y)
   > n5 = GetTimeSineNoise(frequencey=@n2.Z)
   > Return Node!(
    execute
    Rotate:Rotator=MakeRotator(
     Roll=Remap Float(Input=@n3, OldRange=(X=-1,Y=1), NewRange=@n1."X (Roll)", Clamp=true) -> (Output=~)
     Pitch=Remap Float(Input=@n4, OldRange=(X=-1,Y=1), NewRange=@n1."Y (Pitch)", Clamp=true) -> (Output=~)
     Yaw=Remap Float(Input=@n5, OldRange=(X=-1,Y=1), NewRange=@n1."Z (Yaw)", Clamp=true) -> (Output=~)
    )
   )
  ]}
]
variables [
 Delta Time X {type=number}
 Gravity Override {type=Vector}
 Is Backbling Long {type=boolean}
 IsSkyDiveFloatingDown {type=boolean}
 RBAN Settings Out {type=FortRigidBodyAnimNodeParameters}
 SkirtGravityOverride {type=Vector}
 Strip_Leg_base_Rotation {type=Rotator}
 Strip_Rotate_Alpha {type=number}
 Strip_Shoulder_base_Rotation {type=Rotator}
 Strip_Waist_base_Rotation {type=Rotator}
]
